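# Header widget for the Crew subscription progressive gift box, built on CommonUserWidget.
# An Overlay holds a vertical stack: a grid panel with the large Crew logo material, above a centered, wrapping grey "Crew Styles!" description.
# GetCustomBackground returns a fixed soft reference to the Crew gift box background material instance.
# The event graph is empty, so the content is static, with no variables or runtime updates in the Blueprint.
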WidgetBlueprint /Crew/Giftbox/SubscriptionProgressiveGiftBoxHeader
graphs [
 EventGraph {links [], nodes []}
 GetCustomBackground {nodes [
   GetCustomBackground() {is=title K2Node_FunctionEntry}
   > Return Node!(execute, Background:Soft Object Reference=/Crew/Giftbox/Materials/MI_UI_Crew_GBBackground)
  ]}
]
variables []
widgets {
 info {inheritedWidgetCount=0, parentClass=&CommonUserWidget, rootWidgetClass=&Overlay, widgetCount=5}
 widgets [
  Overlay_3 {widgetClassPath=&Overlay, widget=Overlay_3}
  VB_Content {widgetClassPath=&VerticalBox, parent=Overlay_3, widget=VB_Content, slot {class=OverlaySlot}}
  GP_ImagesContainer {
   widgetClassPath=&GridPanel
   parent=VB_Content
   widget=GP_ImagesContainer
   properties {columnFill [1], rowFill [0,0,0]}
   slot {class=VerticalBoxSlot, properties {horizontalAlignment=HAlign_Center, verticalAlignment=VAlign_Top}}
  }
  Img_Crew {
   widgetClassPath=&Image
   parent=GP_ImagesContainer
   widget=Img_Crew
   properties {
    visibility=HitTestInvisible
    brush {
     bIsDynamicallyLoaded=false
     drawAs=Image
     imageType=FullColor
     mirroring=NoMirror
     resourceName=None
     resourceObject=/Game/Athena/UI/Logos/Textures/M_UI_CrewLogo_Large
     tiling=NoTile
     imageSize=<432 168>
     margin=<ltrb 0 0 0 0>
     outlineSettings {
      bUseBrushTransparency=false
      roundingType=HalfHeightRadius
      width=0
      color {colorUseRule=UseColor_Specified, specifiedColor=<rgba 0 0 0 0>}
      cornerRadii=<0 0 0 0>
     }
     tintColor {colorUseRule=UseColor_Specified, specifiedColor=<rgba 1 1 1 1>}
     uVRegion {bIsValid=false, max=<0 0>, min=<0 0>}
    }
   }
   slot {class=GridSlot, properties {horizontalAlignment=HAlign_Center, layer=1, row=2, verticalAlignment=VAlign_Top}}
  }
  Txt_Description {
   widgetClassPath=&CommonTextBlock
   parent=VB_Content
   widget=Txt_Description
   properties {
    autoWrapText=true
    justification=Center
    text="Crew Styles!"
    colorAndOpacity {colorUseRule=UseColor_Specified, specifiedColor=<rgba 0.5776 0.5776 0.5776 1>}
    font {
     bForceMonospaced=false
     bMaterialIsStencil=false
     fontMaterial=null
     fontObject=&NotoSans
     letterSpacing=0
     monospacedWidth=1
     size=24
     skewAmount=0
     typefaceFontName=NotoSans-Regular
     outlineSettings {
      bApplyOutlineToDropShadows=false
      bMiteredCorners=false
      bSeparateFillAlpha=false
      outlineMaterial=null
      outlineSize=0
      outlineColor=<rgba 0 0 0 1>
     }
    }
    shadowOffset=<0 0>
   }
   slot {class=VerticalBoxSlot, properties {verticalAlignment=VAlign_Top, padding=<ltrb 0 8 0 0>}}
  }
 ]
}
