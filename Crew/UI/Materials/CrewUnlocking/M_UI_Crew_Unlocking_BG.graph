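# Translucent UI material for the Crew unlocking background, with a gradient, scrolling tiles and a circular pulse.
# Two-color radial gradient with vignette darkening forms the base, brightened by an SDF ring pulse through color-dodge-style divides.
# Bento texture tiles pan on rotated UVs, with noise-driven time offsets varying their motion.
# Further divides add circle-masked tile highlights and time-seeded random grain.
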Material /Crew/UI/Materials/CrewUnlocking/M_UI_Crew_Unlocking_BG
graph {
 properties {
  bAutomaticallySetUsageInEditor=false
  bUsedWithStaticMesh=false
  blendMode=BLEND_Translucent
  materialDomain=MD_UI
  usedShadingModels=Default Lit
 }
 nodes [
  n1 = Multiply(A=NamedRerouteUsage(), B=Constant3Vector() {constant=<rgba 0.5521 0.5154 1 1>}) {constB=0.2}
  n2 = Power(Base=MaterialFunctionCall!(UVs, Radius=Constant() {r=0.4}) {materialFunction=/FortUILibrary/MaterialFunctions/Shapes/SDFs/MF_SDF_Circle}.SDF)
  n3 = OneMinus(None=Multiply(A=Multiply(A=@n1, B=@n2), B=ScalarParameter() {defaultValue=1, parameterName=Tile_Opacity}) {constB=0})
  n4 = MaterialFunctionCall!(UV, Seed=Multiply(A=Time()) {constB=2}) {materialFunction=&MF_UI_Random}
  Divide(A=Divide(A=NamedRerouteUsage(), B=Max(B=@n3) {constA=0.01, constB=0}), B=OneMinus(None=Multiply(A=@n4.Result) {constB=0.2}))
  Frac(None=Multiply(A=Time()) {constB=0.6})
  NamedRerouteDeclaration(None=ScalarParameter() {parameterName=Pulse Size}) {name=Pulse Size, nodeColor=<rgba 0 0 0 1>}
  n5 = MaterialFunctionCall() {materialFunction=&MF_UI_UVScaling}
  n6 = AppendVector(A=ScalarParameter() {parameterName=Pulse Pos X}, B=ScalarParameter() {parameterName=Pulse Pos Y})
  n7 = NamedRerouteUsage()
  n8 = Power(Base=NamedRerouteUsage())
  n9 = Saturate(None=MaterialFunctionCall(UVs=Add(A=@n5.Scaled to Fit, B=@n6), Size=Multiply(A=@n7) {constB=4}, OutlineThickness=Divide(A=@n8) {constB=4}) {materialFunction=&MF_UI_Circle}.SDFOutline)
  NamedRerouteDeclaration(None=Saturate(None=OneMinus(None=@n9))) {name=Pulse, nodeColor=<rgba 0 0 0 1>}
  n10 = Multiply(A=Constant2Vector() {g=1, r=0.25}, B=Multiply(A=NamedRerouteUsage(), B=NamedRerouteUsage()))
  n11 = Multiply(A=Time(), B=NamedRerouteUsage())
  NamedRerouteDeclaration(None=TextureSample(UVs=Panner(Coordinate=@n10, Time=@n11) {speedX=0.125}, Tex=NamedRerouteUsage()).A) {name=Tiles, nodeColor=<rgba 0 0 0 1>}
  n12 = ScalarParameter() {defaultValue=2, group=Vignette, parameterName=Vignette_Radial_Radius}
  n13 = ScalarParameter() {defaultValue=0.328, group=Background_Gradient, parameterName=Gradient_X_Offset}
  n14 = ScalarParameter() {group=Background_Gradient, parameterName=Gradient_Y_Offset}
  n15 = Add(A=TextureCoordinate(), B=Add(A=Constant2Vector() {g=0.5, r=-0.35}, B=AppendVector(A=@n13, B=@n14)))
  n16 = SmoothStep(Value=OneMinus(None=MaterialFunctionCall(UVs=@n15, Radius=@n12) {materialFunction=&MF_SDF_Circle}.SDF)) {constMax=3.5}
  n17 = ScalarParameter() {defaultValue=1, group=Vignette, parameterName=Vignette_Power}
  NamedRerouteDeclaration(None=Saturate(None=Power(Base=@n16, Exp=@n17))) {name=Vignette_Mask, nodeColor=<rgba 0 0 0 1>}
  n18 = ScalarParameter() {defaultValue=1.6, group=Background_Gradient, parameterName=Background_Radial_Radius}
  n19 = SmoothStep(Value=OneMinus(None=MaterialFunctionCall(UVs=@n15, Radius=@n18) {materialFunction=&MF_SDF_Circle}.SDF)) {constMax=4}
  NamedRerouteDeclaration(None=LinearInterpolate(
    A=VectorParameter() {group=Background_Gradient, parameterName=RGB1, defaultValue=<rgba 0.0322 0.0018 0.0885 1>}.RGB
    B=VectorParameter() {group=Background_Gradient, parameterName=RGB2, defaultValue=<rgba 0.156 0.2385 0.651 1>}.RGB
    Alpha=Saturate(None=@n19)
   )) {name=Background_Gradient, nodeColor=<rgba 0 0 0 1>}
  n20 = MaterialFunctionCall() {materialFunction=&MF_UI_UVScaling}
  NamedRerouteDeclaration(None=@n20.Scaled to Fit) {name=UV, nodeColor=<rgba 0 0 0 1>}
  NamedRerouteDeclaration(None=@n20.ScaleY) {name=UV-Scale-Y, nodeColor=<rgba 0 0 0 1>}
  n21 = VectorParameter() {group=Background_Gradient, parameterName=RGB1, defaultValue=<rgba 0.0322 0.0018 0.0885 1>}
  n22 = Multiply(A=@n21.RGB, B=ScalarParameter() {defaultValue=0.25, group=Vignette, parameterName=Vignette_Edge_Darken}) {constB=0.25}
  n23 = NamedRerouteUsage()
  NamedRerouteDeclaration(None=LinearInterpolate(A=LinearInterpolate(A=@n22, B=@n21.RGB, Alpha=@n23), B=@n23, Alpha=NamedRerouteUsage())) {name=Base Gradient, nodeColor=<rgba 0 0 0 1>}
  n24 = NamedRerouteUsage()
  n25 = NamedRerouteUsage()
  n26 = LinearInterpolate(
   A=LinearInterpolate(A=Constant3Vector() {constant=<rgba 0.0723 0.0595 0.1301 1>}, B=Constant3Vector() {constant=<rgba 0.15 0.1144 0.4678 1>}, Alpha=@n25)
   B=Constant3Vector() {constant=<rgba 0.7529 0.7991 1 1>}
   Alpha=@n25
  )
  n27 = Max(B=Saturate(None=Multiply(A=@n26) {constB=1.1}))
  n28 = NamedRerouteUsage()
  n29 = Divide(A=@n24, B=Multiply(A=OneMinus(None=@n27), B=SmoothStep(Value=@n28) {constMax=0.2})) {constB=1}
  NamedRerouteDeclaration(None=LinearInterpolate(A=@n24, B=Saturate(None=@n29), Alpha=ScalarParameter() {parameterName=Pulse Opacity}) {constAlpha=1}) {name=Pulse + Gradient, nodeColor=<rgba 0 0 0 1>}
  NamedRerouteDeclaration(None=Rotator(Coordinate=NamedRerouteUsage(), Time=ScalarParameter() {defaultValue=0.08, group=Global, parameterName=Tile_Angle}) {speed=6.2832}) {name=Screen_Angle, nodeColor=<rgba 0 0 0 1>}
  NamedRerouteDeclaration(None=ScalarParameter() {defaultValue=1.675, group=Tiles_Foreground, parameterName=Tile-Scale-Foreground}) {name=FG_Tile_Scale, nodeColor=<rgba 0 0 0 1>}
  NamedRerouteDeclaration(None=TextureObjectParameter() {parameterName=CMS_Crew_Bento, texture=/Game/Athena/UI/Subscriptions/Art/Textures/T_CrewUpsell_Bento}) {name=Texture Object, nodeColor=<rgba 0 0 0 1>}
  n30 = Multiply(A=NamedRerouteUsage(), B=AppendVector(A=Constant() {r=0.5}, B=NamedRerouteUsage()))
  n31 = AppendVector(
   A=ScalarParameter() {defaultValue=0.25, group=Tiles_Foreground, parameterName=FG_Time_Offset_X-Tiling}
   B=ScalarParameter() {defaultValue=1, group=Tiles_Foreground, parameterName=FG_Time_Offset_Y-Tiling}
  )
  n32 = ScalarParameter() {defaultValue=-0.5, group=Tiles_Foreground, parameterName=FG_Time_Offset_X-Offsets}
  n33 = Round(None=ScalarParameter() {defaultValue=4, group=Tiles_Foreground, parameterName=FG_Time_Offset_Y-Offsets})
  n34 = MaterialFunctionCall(Float2=Add(A=ScalarParameter() {defaultValue=0.0625, group=Tiles_Foreground, parameterName=FG_Time_Mask_Offset}, B=@n30)) {materialFunction=&BreakOutFloat2Components}
  n35 = ConstantBiasScale(None=Add(A=MaterialFunctionCall(Value=@n34.G) {materialFunction=&LinearSine}.Linear Sine) {constB=-0.5}) {bias=-0.125, scale=12}
  n36 = MaterialFunctionCall!(
   PrimaryTextureCoordinates=Add(A=Multiply(A=@n30, B=@n31), B=AppendVector(A=@n32, B=@n33))
   2ndOrderTextureCoordinates
   RandomSeed
   NoiseDensityBias=LinearInterpolate(
    A=ScalarParameter() {defaultValue=1, group=Tiles_Foreground, parameterName=FG_Time_Mask_Upper_Range}
    B=ScalarParameter() {defaultValue=0.125, group=Tiles_Foreground, parameterName=FG_Time_Mask_Lower_Range}
    Alpha=Saturate(None=@n35)
   ) {constA=1, constB=0.5}
   Bypass Noise Influence Crosstalk
  ) {materialFunction=/Game/Effects/Fort_Effects/Materials/_Functions/MF_RandomUVSample}
  NamedRerouteDeclaration(None=Multiply(
    A=ConstantBiasScale(None=@n36.NoiseWithRange) {bias=0.5, scale=-2}
    B=ScalarParameter() {defaultValue=0.05, group=Tiles_Foreground, parameterName=Tile_FG_Time_Factor}
   )) {name=FG_Time_Offset, nodeColor=<rgba 0 0 0 1>}
 ]
}
