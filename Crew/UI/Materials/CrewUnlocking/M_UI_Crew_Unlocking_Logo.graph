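# Translucent UI material that reveals the Crew wordmark logo with a pixelated transition for the Crew unlocking screen.
# A Reveal Amount parameter drives a feathered mask across tilted UVs, with adjustable angle.
# A particle grid with random per-cell values sizes rounded squares that break the logo into pixels.
# Masked logo layers combine via Max, with intensity remapped into a dark purple, periwinkle and white color ramp.
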Material /Crew/UI/Materials/CrewUnlocking/M_UI_Crew_Unlocking_Logo
graph {
 properties {
  bAutomaticallySetUsageInEditor=false
  bUsedWithStaticMesh=false
  blendMode=BLEND_Translucent
  materialDomain=MD_UI
  usedShadingModels=Default Lit
 }
 nodes [
  n1 = NamedRerouteUsage()
  n2 = NamedRerouteUsage()
  n3 = NamedRerouteUsage()
  n4 = Max(A=Multiply(A=NamedRerouteUsage(), B=SmoothStep(Min=@n1, Value=@n2)), B=Multiply(A=OneMinus(None=@n3), B=NamedRerouteUsage()))
  LinearInterpolate(
   A=Constant3Vector() {constant=<rgba 0.0331 0.0144 0.1144 1>}
   B=LinearInterpolate(
    A=Constant3Vector() {constant=<rgba 0.2623 0.2705 0.8963 1>}
    B=Constant3Vector() {constant=<rgba 1 1 1 1>}
    Alpha=Power(Base=@n4) {constExponent=4}
   )
   Alpha=SmoothStep(Value=@n4) {constMax=0.4}
  )
  MaterialFunctionCall(Speed=Constant() {r=0.2}) {materialFunction=&DebugTimeSine}
  n5 = NamedRerouteUsage()
  Max(A=NamedRerouteUsage(), B=Multiply(A=OneMinus(None=@n5), B=NamedRerouteUsage()))
  n6 = ScalarParameter() {defaultValue=540, group=Texture, parameterName="Width (X)", sortPriority=0}
  n7 = ScalarParameter() {defaultValue=160, group=Texture, parameterName="Height (Y)", sortPriority=0}
  n8 = TextureSampleParameter2D(UVs=MaterialFunctionCall(TextureSize=AppendVector(A=@n6, B=@n7)) {materialFunction=&MF_UI_UVScaling}.Scaled to Fit) {
   group=Texture
   parameterName=Logo_Texture
   samplerSource=SSM_Clamp_WorldGroupSettings
   samplerType=SAMPLERTYPE_Alpha
   texture=/Game/Athena/UI/Logos/Textures/T_Crew_Workmark_L
  }
  NamedRerouteDeclaration(None=@n8.A) {name=LogoTexture, nodeColor=<rgba 0 0 0 1>}
  n9 = ScalarParameter() {defaultValue=1, parameterName=Reveal Amount}
  NamedRerouteDeclaration(None=Multiply(A=OneMinus(None=@n9)) {constB=2}) {name=Reveal Amount, nodeColor=<rgba 0 0 0 1>}
  n10 = NamedRerouteUsage()
  n11 = ScalarParameter() {defaultValue=0.8, parameterName=Reveal Feather}
  n12 = ComponentMask(None=NamedRerouteUsage()) {r=true}
  NamedRerouteDeclaration(None=SmoothStep(Min=@n10, Max=Subtract(A=@n10, B=@n11) {constB=0.1}, Value=OneMinus(None=@n12))) {name=Reveal Mask, nodeColor=<rgba 0 0 0 1>}
  n13 = MaterialFunctionCall() {materialFunction=&MF_UI_UVScaling}
  n14 = ScalarParameter() {defaultValue=0.05, parameterName=Reveal Angle}
  n15 = Multiply(A=@n14) {constB=10}
  NamedRerouteDeclaration(None=MaterialFunctionCall(
    UVs=MaterialFunctionCall(UVs=@n13.Scaled to Fit, "Rotation Angle (0-1)"=@n14) {materialFunction=&CustomRotator}.Rotated Values
    Texture Scale=Add(A=@n15)
   ) {materialFunction=&ScaleUVsByCenter}.UVs) {name=TiltedUV, nodeColor=<rgba 0 0 0 1>}
  n16 = ScalarParameter() {defaultValue=30, parameterName=SquareAmount}
  n17 = Frac(None=Multiply(A=NamedRerouteUsage(), B=@n16) {constB=12})
  n18 = ScalarParameter() {defaultValue=0.4, group=Layout, parameterName=CornerRadius, sortPriority=6}
  n19 = MaterialFunctionCall!(UV=NamedRerouteUsage(), Global Offset, Global Rotation, Tiling=@n16, Cell Offset, Cell Scale, Cell Rotation, Random Seed=Constant() {r=4}) {materialFunction=/FortUILibrary/MaterialFunctions/Basics/MF_UI_ParticleGrid}
  n20 = Multiply(A=Step(Y=NamedRerouteUsage(), X=@n19.Random Cell Values)) {constB=0.8}
  NamedRerouteDeclaration(None=Multiply(
    A=NamedRerouteUsage()
    B=MaterialFunctionCall(UV=@n17, Size=@n20, Corner Radius=@n18) {materialFunction=&MF_UI_CoreUI_RoundedMask}.Shape Mask
   )) {name=Pixeled Logo, nodeColor=<rgba 0 0 0 1>}
  NamedRerouteDeclaration(None=@n19.Random Cell Values) {name=Random Cell Value, nodeColor=<rgba 0 0 0 1>}
  Saturate()
 ]
}
