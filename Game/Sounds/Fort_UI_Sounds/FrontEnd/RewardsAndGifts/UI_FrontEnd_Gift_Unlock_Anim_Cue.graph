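# UI sound cue for the front-end gift unlock animation in rewards and gifts.
# It plays a single member-benefits item-received unlock wave after a short fixed delay, boosted above default volume and routed to the UI sound class.
# A concurrency override allows only one instance, stopping the oldest, with a retrigger time set, alongside a referenced front-end menu concurrency set.
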SoundCue /Game/Sounds/Fort_UI_Sounds/FrontEnd/RewardsAndGifts/UI_FrontEnd_Gift_Unlock_Anim_Cue
root=Delay_0
nodes [
 {id=Delay_0, children [WavePlayer_0], properties {delayMax=0.1, delayMin=0.1}}
 {
  id=WavePlayer_0
  properties {soundWaveAssetPtr=/Game/Sounds/Fort_UI_Sounds/FrontEnd/RewardsAndGifts/wav/UI_ItemRecieved_MemberBenifits_Unlock}
 }
]
properties {
 bOverrideConcurrency=true
 maxDistance=2097152
 soundClassObject=&Fort_UI
 volumeMultiplier=1.6
 concurrencyOverrides {maxCount=1, resolutionRule=StopOldest, retriggerTime=2}
 concurrencySet [/Game/Sounds/Concurrency_Rules/FrontEnd_UI_Menu_Max_03_Stop_Oldest]
}
