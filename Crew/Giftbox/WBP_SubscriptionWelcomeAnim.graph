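# Fortnite Crew welcome screen (AthenaUpgradedPreMessage) that animates the Crew logo and "Member benefits" over a material-driven background.
# On activation it plays UnlockAnimation_Short with an unlock sound and music-ducking bus mix, released after a delay.
# A delayed disappear sound accompanies the exit, and animation completion calls OnPresentationFinished.
# The longer UnlockAnimation, which also reveals four pass badges and slogans, is unused by this graph; the short one keeps them hidden.
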WidgetBlueprint /Crew/Giftbox/WBP_SubscriptionWelcomeAnim
animations [
 UnlockAnimation {
  binding_count=12
  length_seconds=7.5
  object=&self:UnlockAnimation
  sequence {
   bindings [
    Crew_Background {
     id=#1
     possessed_class=Image
     tracks [
      {
       type=MovieSceneWidgetMaterialTrack
       sections [{
         type=MovieSceneParameterSection
         end_frame=225
         start_frame=0
         channels [
          Pulse Opacity {type=MovieSceneScriptingFloatChannel, keys [[-0.0005,0],[24.746,0],[26,1],[212,0],[213,1],[219,0]]}
          Pulse Size {type=MovieSceneScriptingFloatChannel, keys [[-0.0005,0],[26,0],[129,0.5784],[212,0.8226],[222,0]]}
          Vignette_Power {type=MovieSceneScriptingFloatChannel, keys [[-0.0005,16.7782],[7,3.4653],[8.746,1],[205,1],[219,3.4653],[222,16.7782]]}
          Vignette_Edge_Darken {type=MovieSceneScriptingFloatChannel, keys [[-0.0005,0],[7,0.2109],[8.746,0.25],[205,0.25],[219,0.2109],[222,0]]}
          Background_Radial_Radius {
           type=MovieSceneScriptingFloatChannel
           keys [[-0.0005,-0.6],[5.746,1.2],[7,1.2563],[28,1.3811],[29.254,2],[30.254,5.0968],[45.254,1.4],[205,1.8],[219,1.2563],[222,-0.6]]
          }
          Tile_Opacity {type=MovieSceneScriptingFloatChannel, keys [[-0.0005,0],[7.9995,0],[29.254,1]]}
          Tile-Scale-Foreground {type=MovieSceneScriptingFloatChannel}
          Gradient_Y_Offset {type=MovieSceneScriptingFloatChannel, keys [[87,-0.13],[109,0]]}
         ]
        }]
      }
      {
       type=MovieSceneColorTrack
       display_name=Color and Opacity
       sections [{
         type=MovieSceneColorSection
         end_frame=225
         start_frame=0
         channels [
          Color.R {type=MovieSceneScriptingFloatChannel, keys [[0,0],[6.032,1],[210,1],[225,0]]}
          Color.G {type=MovieSceneScriptingFloatChannel, keys [[0,0],[6.032,1],[210,1],[225,0]]}
          Color.B {type=MovieSceneScriptingFloatChannel, keys [[0,0],[6.032,1],[210,1],[225,0]]}
          Color.A {type=MovieSceneScriptingFloatChannel, keys [[0,0],[6.032,1],[210,1],[225,0]]}
         ]
        }]
      }
     ]
    }
    Crew_Logo {
     id=#2
     possessed_class=Image
     tracks [
      {
       type=MovieSceneWidgetMaterialTrack
       sections [{
         type=MovieSceneParameterSection
         end_frame=225
         start_frame=0
         channels [Reveal Amount {type=MovieSceneScriptingFloatChannel, keys [[0,0],[25.254,1]]}]
        }]
      }
      {
       type=MovieSceneFloatTrack
       display_name=Render Opacity
       sections [{
         type=MovieSceneFloatSection
         end_frame=225
         start_frame=0
         channels [None {type=MovieSceneScriptingFloatChannel, keys [[0,1],[203,1],[220,0]]}]
        }]
      }
     ]
    }
    Figment_Badge {
     id=#3
     possessed_class=WBP_SubscriptionWelcome_Badge_C
     tracks [
      {
       type=MovieSceneFloatTrack
       display_name=Render Opacity
       sections [{type=MovieSceneFloatSection, end_frame=172, start_frame=0, channels [None {type=MovieSceneScriptingFloatChannel, keys [[100,0],[109,1]]}]}]
      }
      {
       type=MovieScene2DTransformTrack
       sections [{
         type=MovieScene2DTransformSection
         end_frame=172
         start_frame=0
         channels [
          Translation.X {type=MovieSceneScriptingFloatChannel, const [0,1]}
          Translation.Y {type=MovieSceneScriptingFloatChannel, const [0,1]}
          Angle {type=MovieSceneScriptingFloatChannel, const [0,1]}
          Scale.X {type=MovieSceneScriptingFloatChannel, keys [[102,1.9194],[111,1]]}
          Scale.Y {type=MovieSceneScriptingFloatChannel, keys [[102,1.9194],[111,1]]}
          Shear.X {type=MovieSceneScriptingFloatChannel, const [0,1]}
          Shear.Y {type=MovieSceneScriptingFloatChannel, const [0,1]}
         ]
        }]
      }
      {
       type=MovieSceneWidgetMaterialTrack
       sections [{
         type=MovieSceneParameterSection
         end_frame=225
         start_frame=0
         channels [Shine Amount {type=MovieSceneScriptingFloatChannel, keys [[132,0],[156,1]]}]
        }]
      }
     ]
    }
    Music_Badge {
     id=#4
     possessed_class=WBP_SubscriptionWelcome_Badge_C
     tracks [
      {
       type=MovieSceneFloatTrack
       display_name=Render Opacity
       sections [{type=MovieSceneFloatSection, end_frame=177, start_frame=0, channels [None {type=MovieSceneScriptingFloatChannel, keys [[103,0],[112,1]]}]}]
      }
      {
       type=MovieScene2DTransformTrack
       sections [{
         type=MovieScene2DTransformSection
         end_frame=177
         start_frame=0
         channels [
          Translation.X {type=MovieSceneScriptingFloatChannel, const [0,1]}
          Translation.Y {type=MovieSceneScriptingFloatChannel, const [0,1]}
          Angle {type=MovieSceneScriptingFloatChannel, const [0,1]}
          Scale.X {type=MovieSceneScriptingFloatChannel, keys [[105,1.9194],[114,1]]}
          Scale.Y {type=MovieSceneScriptingFloatChannel, keys [[105,1.9194],[114,1]]}
          Shear.X {type=MovieSceneScriptingFloatChannel, const [0,1]}
          Shear.Y {type=MovieSceneScriptingFloatChannel, const [0,1]}
         ]
        }]
      }
      {
       type=MovieSceneWidgetMaterialTrack
       sections [{
         type=MovieSceneParameterSection
         end_frame=225
         start_frame=0
         channels [Shine Amount {type=MovieSceneScriptingFloatChannel, keys [[140,0],[164,1]]}]
        }]
      }
     ]
    }
    BP_Badge {
     id=#5
     possessed_class=WBP_SubscriptionWelcome_Badge_C
     tracks [
      {
       type=MovieSceneFloatTrack
       display_name=Render Opacity
       sections [{type=MovieSceneFloatSection, end_frame=172, start_frame=0, channels [None {type=MovieSceneScriptingFloatChannel, keys [[97,0],[106,1]]}]}]
      }
      {
       type=MovieScene2DTransformTrack
       sections [{
         type=MovieScene2DTransformSection
         end_frame=169
         start_frame=0
         channels [
          Translation.X {type=MovieSceneScriptingFloatChannel, const [0,1]}
          Translation.Y {type=MovieSceneScriptingFloatChannel, const [0,1]}
          Angle {type=MovieSceneScriptingFloatChannel, const [0,1]}
          Scale.X {type=MovieSceneScriptingFloatChannel, keys [[99,1.9194],[108,1]]}
          Scale.Y {type=MovieSceneScriptingFloatChannel, keys [[99,1.9194],[108,1]]}
          Shear.X {type=MovieSceneScriptingFloatChannel, const [0,1]}
          Shear.Y {type=MovieSceneScriptingFloatChannel, const [0,1]}
         ]
        }]
      }
      {
       type=MovieSceneWidgetMaterialTrack
       sections [{
         type=MovieSceneParameterSection
         end_frame=225
         start_frame=0
         channels [Shine Amount {type=MovieSceneScriptingFloatChannel, keys [[126,0],[150,1]]}]
        }]
      }
     ]
    }
    SloganText_2 {
     id=#6
     possessed_class=CommonTextBlock
     tracks [
      {
       type=MovieScene2DTransformTrack
       sections [{
         type=MovieScene2DTransformSection
         end_frame=197
         start_frame=0
         channels [
          Translation.X {type=MovieSceneScriptingFloatChannel}
          Translation.Y {type=MovieSceneScriptingFloatChannel}
          Angle {type=MovieSceneScriptingFloatChannel}
          Scale.X {type=MovieSceneScriptingFloatChannel, keys [[120,2],[128,1]]}
          Scale.Y {type=MovieSceneScriptingFloatChannel, keys [[120,2],[128,1]]}
          Shear.X {type=MovieSceneScriptingFloatChannel}
          Shear.Y {type=MovieSceneScriptingFloatChannel}
         ]
        }]
      }
      {
       type=MovieSceneFloatTrack
       display_name=Render Opacity
       sections [{
         type=MovieSceneFloatSection
         end_frame=225
         start_frame=0
         channels [None {type=MovieSceneScriptingFloatChannel, keys [[120,0],[125,1],[200,1],[205,0]]}]
        }]
      }
     ]
    }
    Lego_Badge {
     id=#7
     possessed_class=WBP_SubscriptionWelcome_Badge_C
     tracks [
      {
       type=MovieSceneFloatTrack
       display_name=Render Opacity
       sections [{type=MovieSceneFloatSection, end_frame=182, start_frame=0, channels [None {type=MovieSceneScriptingFloatChannel, keys [[107,0],[116,1]]}]}]
      }
      {
       type=MovieScene2DTransformTrack
       sections [{
         type=MovieScene2DTransformSection
         end_frame=182
         start_frame=0
         channels [
          Translation.X {type=MovieSceneScriptingFloatChannel, const [0,1]}
          Translation.Y {type=MovieSceneScriptingFloatChannel, const [0,1]}
          Angle {type=MovieSceneScriptingFloatChannel, const [0,1]}
          Scale.X {type=MovieSceneScriptingFloatChannel, keys [[109,1.9194],[118,1]]}
          Scale.Y {type=MovieSceneScriptingFloatChannel, keys [[109,1.9194],[118,1]]}
          Shear.X {type=MovieSceneScriptingFloatChannel, const [0,1]}
          Shear.Y {type=MovieSceneScriptingFloatChannel, const [0,1]}
         ]
        }]
      }
      {
       type=MovieSceneWidgetMaterialTrack
       sections [{
         type=MovieSceneParameterSection
         end_frame=225
         start_frame=0
         channels [Shine Amount {type=MovieSceneScriptingFloatChannel, keys [[150,0],[174,1]]}]
        }]
      }
     ]
    }
    Slogantext_1 {
     id=#8
     possessed_class=CommonTextBlock
     tracks [
      {
       type=MovieScene2DTransformTrack
       sections [{
         type=MovieScene2DTransformSection
         end_frame=183
         start_frame=0
         channels [
          Translation.X {type=MovieSceneScriptingFloatChannel}
          Translation.Y {type=MovieSceneScriptingFloatChannel}
          Angle {type=MovieSceneScriptingFloatChannel}
          Scale.X {type=MovieSceneScriptingFloatChannel, keys [[117,2],[125,1]]}
          Scale.Y {type=MovieSceneScriptingFloatChannel, keys [[117,2],[125,1]]}
          Shear.X {type=MovieSceneScriptingFloatChannel}
          Shear.Y {type=MovieSceneScriptingFloatChannel}
         ]
        }]
      }
      {
       type=MovieSceneFloatTrack
       display_name=Render Opacity
       sections [{
         type=MovieSceneFloatSection
         end_frame=225
         start_frame=0
         channels [None {type=MovieSceneScriptingFloatChannel, keys [[117,0],[122,1],[202,1],[206.9995,0]]}]
        }]
      }
     ]
    }
    Crew_Subheading {
     id=#9
     possessed_class=CommonTextBlock
     tracks [
      {
       type=MovieScene2DTransformTrack
       sections [{
         type=MovieScene2DTransformSection
         end_frame=105
         start_frame=0
         channels [
          Translation.X {type=MovieSceneScriptingFloatChannel}
          Translation.Y {type=MovieSceneScriptingFloatChannel, keys [[47,-40],[54,0]]}
          Angle {type=MovieSceneScriptingFloatChannel}
          Scale.X {type=MovieSceneScriptingFloatChannel}
          Scale.Y {type=MovieSceneScriptingFloatChannel}
          Shear.X {type=MovieSceneScriptingFloatChannel}
          Shear.Y {type=MovieSceneScriptingFloatChannel}
         ]
        }]
      }
      {
       type=MovieSceneColorTrack
       display_name=Color and Opacity
       sections [{
         type=MovieSceneColorSection
         end_frame=156
         start_frame=0
         channels [
          Color.R {type=MovieSceneScriptingFloatChannel, const [0.5457,2]}
          Color.G {type=MovieSceneScriptingFloatChannel, const [0.6105,2]}
          Color.B {type=MovieSceneScriptingFloatChannel, const [1,2]}
          Color.A {type=MovieSceneScriptingFloatChannel, keys [[47,0],[54,1],[92,1],[95,0]]}
         ]
        }]
      }
     ]
    }
    ContentBox {
     id=#10
     possessed_class=Overlay
     tracks [{
       type=MovieScene2DTransformTrack
       sections [{
         type=MovieScene2DTransformSection
         end_frame=225
         start_frame=0
         channels [
          Translation.X {type=MovieSceneScriptingFloatChannel, const [0,1]}
          Translation.Y {type=MovieSceneScriptingFloatChannel, const [0,1]}
          Angle {type=MovieSceneScriptingFloatChannel, const [0,1]}
          Scale.X {type=MovieSceneScriptingFloatChannel, keys [[0,1],[193.746,1],[212.765,0.5]]}
          Scale.Y {type=MovieSceneScriptingFloatChannel, keys [[0,1],[193.746,1],[212.765,0.5]]}
          Shear.X {type=MovieSceneScriptingFloatChannel, const [0,1]}
          Shear.Y {type=MovieSceneScriptingFloatChannel, const [0,1]}
         ]
        }]
      }]
    }
    LogoOverlay {
     id=#11
     possessed_class=Overlay
     tracks [{
       type=MovieScene2DTransformTrack
       sections [{
         type=MovieScene2DTransformSection
         end_frame=225
         start_frame=0
         channels [
          Translation.X {type=MovieSceneScriptingFloatChannel}
          Translation.Y {type=MovieSceneScriptingFloatChannel, keys [[93,0],[109,-240],[205,-240],[216,0]]}
          Angle {type=MovieSceneScriptingFloatChannel}
          Scale.X {type=MovieSceneScriptingFloatChannel, keys [[0,1.8148],[25,1.2],[93,1.2],[109,0.55]]}
          Scale.Y {type=MovieSceneScriptingFloatChannel, keys [[0,1.8148],[25,1.2],[93,1.2],[109,0.55]]}
          Shear.X {type=MovieSceneScriptingFloatChannel}
          Shear.Y {type=MovieSceneScriptingFloatChannel}
         ]
        }]
      }]
    }
    PassBadges {
     id=#12
     possessed_class=HorizontalBox
     tracks [
      {
       type=MovieSceneFloatTrack
       display_name=Render Opacity
       sections [{
         type=MovieSceneFloatSection
         end_frame=225
         start_frame=0
         channels [None {type=MovieSceneScriptingFloatChannel, keys [[0,0],[101,0],[104,1],[203,1],[210,0]]}]
        }]
      }
      {type=MovieScene2DTransformTrack, sections []}
     ]
    }
   ]
   display_rate {numerator=30}
  }
 }
 UnlockAnimation_Short {
  binding_count=12
  length_seconds=3.5
  object=&self:UnlockAnimation_Short
  sequence {
   bindings [
    Crew_Background {
     id=#1
     possessed_class=Image
     tracks [
      {
       type=MovieSceneColorTrack
       display_name=Color and Opacity
       sections [{
         type=MovieSceneColorSection
         end_frame=105
         start_frame=0
         channels [
          Color.R {type=MovieSceneScriptingFloatChannel, keys [[0,0],[6.032,1],[93,1],[108,0]]}
          Color.G {type=MovieSceneScriptingFloatChannel, keys [[0,0],[6.032,1],[93,1],[108,0]]}
          Color.B {type=MovieSceneScriptingFloatChannel, keys [[0,0],[6.032,1],[93,1],[108,0]]}
          Color.A {type=MovieSceneScriptingFloatChannel, keys [[0,0],[6.032,1],[93,1],[108,0]]}
         ]
        }]
      }
      {
       type=MovieSceneWidgetMaterialTrack
       sections [{
         type=MovieSceneParameterSection
         end_frame=105
         start_frame=0
         channels [
          Pulse Opacity {type=MovieSceneScriptingFloatChannel, keys [[-0.0005,0],[24.746,0],[26,1],[94,0],[95,1],[101,0]]}
          Pulse Size {type=MovieSceneScriptingFloatChannel, keys [[-0.0005,0],[26,0],[87,0.5784],[94,0.8226],[104,0]]}
          Vignette_Power {type=MovieSceneScriptingFloatChannel, keys [[-0.0005,16.7782],[7,3.4653],[8.746,1],[87,1],[101,3.4653],[104,16.7782]]}
          Vignette_Edge_Darken {type=MovieSceneScriptingFloatChannel, keys [[-0.0005,0],[7,0.2109],[8.746,0.25],[87,0.25],[101,0.2109],[104,0]]}
          Background_Radial_Radius {
           type=MovieSceneScriptingFloatChannel
           keys [[-0.0005,-0.6],[5.746,1.2],[7,1.2563],[28,1.3811],[29.254,2],[30.254,5.0968],[45.254,1.4],[87,1.8],[101,1.2563],[104,-0.6]]
          }
          Tile_Opacity {type=MovieSceneScriptingFloatChannel, keys [[-0.0005,0],[7.9995,0],[29.254,1]]}
          Tile-Scale-Foreground {type=MovieSceneScriptingFloatChannel}
          Gradient_Y_Offset {type=MovieSceneScriptingFloatChannel, keys [[80,-0.13],[102,0]]}
         ]
        }]
      }
     ]
    }
    Crew_Logo {
     id=#2
     possessed_class=Image
     tracks [
      {
       type=MovieSceneWidgetMaterialTrack
       sections [{
         type=MovieSceneParameterSection
         end_frame=105
         start_frame=0
         channels [Reveal Amount {type=MovieSceneScriptingFloatChannel, keys [[0,0],[25.254,1]]}]
        }]
      }
      {
       type=MovieSceneFloatTrack
       display_name=Render Opacity
       sections [{
         type=MovieSceneFloatSection
         end_frame=105
         start_frame=0
         channels [None {type=MovieSceneScriptingFloatChannel, keys [[0,1],[84,1],[101,0]]}]
        }]
      }
     ]
    }
    Figment_Badge {
     id=#3
     possessed_class=WBP_SubscriptionWelcome_Badge_C
     tracks [
      {
       type=MovieSceneFloatTrack
       display_name=Render Opacity
       sections [{type=MovieSceneFloatSection, end_frame=80, start_frame=0, channels [None {type=MovieSceneScriptingFloatChannel, const [0,1]}]}]
      }
      {
       type=MovieScene2DTransformTrack
       sections [{
         type=MovieScene2DTransformSection
         end_frame=80
         start_frame=0
         channels [
          Translation.X {type=MovieSceneScriptingFloatChannel, const [0,1]}
          Translation.Y {type=MovieSceneScriptingFloatChannel, const [0,1]}
          Angle {type=MovieSceneScriptingFloatChannel, const [0,1]}
          Scale.X {type=MovieSceneScriptingFloatChannel, const [1.9194,1]}
          Scale.Y {type=MovieSceneScriptingFloatChannel, const [1.9194,1]}
          Shear.X {type=MovieSceneScriptingFloatChannel, const [0,1]}
          Shear.Y {type=MovieSceneScriptingFloatChannel, const [0,1]}
         ]
        }]
      }
      {
       type=MovieSceneWidgetMaterialTrack
       sections [{type=MovieSceneParameterSection, end_frame=105, start_frame=0, channels [Shine Amount {type=MovieSceneScriptingFloatChannel, default=0}]}]
      }
     ]
    }
    Music_Badge {
     id=#4
     possessed_class=WBP_SubscriptionWelcome_Badge_C
     tracks [
      {
       type=MovieSceneFloatTrack
       display_name=Render Opacity
       sections [{type=MovieSceneFloatSection, end_frame=82, start_frame=0, channels [None {type=MovieSceneScriptingFloatChannel, const [0,1]}]}]
      }
      {
       type=MovieScene2DTransformTrack
       sections [{
         type=MovieScene2DTransformSection
         end_frame=82
         start_frame=0
         channels [
          Translation.X {type=MovieSceneScriptingFloatChannel, const [0,1]}
          Translation.Y {type=MovieSceneScriptingFloatChannel, const [0,1]}
          Angle {type=MovieSceneScriptingFloatChannel, const [0,1]}
          Scale.X {type=MovieSceneScriptingFloatChannel, const [1.9194,1]}
          Scale.Y {type=MovieSceneScriptingFloatChannel, const [1.9194,1]}
          Shear.X {type=MovieSceneScriptingFloatChannel, const [0,1]}
          Shear.Y {type=MovieSceneScriptingFloatChannel, const [0,1]}
         ]
        }]
      }
      {
       type=MovieSceneWidgetMaterialTrack
       sections [{type=MovieSceneParameterSection, end_frame=105, start_frame=0, channels [Shine Amount {type=MovieSceneScriptingFloatChannel, default=0}]}]
      }
     ]
    }
    BP_Badge {
     id=#5
     possessed_class=WBP_SubscriptionWelcome_Badge_C
     tracks [
      {
       type=MovieSceneFloatTrack
       display_name=Render Opacity
       sections [{type=MovieSceneFloatSection, end_frame=80, start_frame=0, channels [None {type=MovieSceneScriptingFloatChannel, const [0,1]}]}]
      }
      {
       type=MovieScene2DTransformTrack
       sections [{
         type=MovieScene2DTransformSection
         end_frame=78
         start_frame=0
         channels [
          Translation.X {type=MovieSceneScriptingFloatChannel, const [0,1]}
          Translation.Y {type=MovieSceneScriptingFloatChannel, const [0,1]}
          Angle {type=MovieSceneScriptingFloatChannel, const [0,1]}
          Scale.X {type=MovieSceneScriptingFloatChannel, const [1.9194,1]}
          Scale.Y {type=MovieSceneScriptingFloatChannel, const [1.9194,1]}
          Shear.X {type=MovieSceneScriptingFloatChannel, const [0,1]}
          Shear.Y {type=MovieSceneScriptingFloatChannel, const [0,1]}
         ]
        }]
      }
      {
       type=MovieSceneWidgetMaterialTrack
       sections [{type=MovieSceneParameterSection, end_frame=105, start_frame=0, channels [Shine Amount {type=MovieSceneScriptingFloatChannel, default=0}]}]
      }
     ]
    }
    SloganText_2 {
     id=#6
     possessed_class=CommonTextBlock
     tracks [
      {
       type=MovieScene2DTransformTrack
       sections [{
         type=MovieScene2DTransformSection
         end_frame=105
         start_frame=0
         channels [
          Translation.X {type=MovieSceneScriptingFloatChannel}
          Translation.Y {type=MovieSceneScriptingFloatChannel}
          Angle {type=MovieSceneScriptingFloatChannel}
          Scale.X {type=MovieSceneScriptingFloatChannel, const [2,1]}
          Scale.Y {type=MovieSceneScriptingFloatChannel, const [2,1]}
          Shear.X {type=MovieSceneScriptingFloatChannel}
          Shear.Y {type=MovieSceneScriptingFloatChannel}
         ]
        }]
      }
      {
       type=MovieSceneFloatTrack
       display_name=Render Opacity
       sections [{type=MovieSceneFloatSection, end_frame=105, start_frame=0, channels [None {type=MovieSceneScriptingFloatChannel, const [0,1]}]}]
      }
     ]
    }
    Lego_Badge {
     id=#7
     possessed_class=WBP_SubscriptionWelcome_Badge_C
     tracks [
      {
       type=MovieSceneFloatTrack
       display_name=Render Opacity
       sections [{type=MovieSceneFloatSection, end_frame=84, start_frame=0, channels [None {type=MovieSceneScriptingFloatChannel, const [0,1]}]}]
      }
      {
       type=MovieScene2DTransformTrack
       sections [{
         type=MovieScene2DTransformSection
         end_frame=84
         start_frame=0
         channels [
          Translation.X {type=MovieSceneScriptingFloatChannel, const [0,1]}
          Translation.Y {type=MovieSceneScriptingFloatChannel, const [0,1]}
          Angle {type=MovieSceneScriptingFloatChannel, const [0,1]}
          Scale.X {type=MovieSceneScriptingFloatChannel, const [1.9194,1]}
          Scale.Y {type=MovieSceneScriptingFloatChannel, const [1.9194,1]}
          Shear.X {type=MovieSceneScriptingFloatChannel, const [0,1]}
          Shear.Y {type=MovieSceneScriptingFloatChannel, const [0,1]}
         ]
        }]
      }
      {
       type=MovieSceneWidgetMaterialTrack
       sections [{type=MovieSceneParameterSection, end_frame=105, start_frame=0, channels [Shine Amount {type=MovieSceneScriptingFloatChannel, default=0}]}]
      }
     ]
    }
    Slogantext_1 {
     id=#8
     possessed_class=CommonTextBlock
     tracks [
      {
       type=MovieScene2DTransformTrack
       sections [{
         type=MovieScene2DTransformSection
         end_frame=105
         start_frame=0
         channels [
          Translation.X {type=MovieSceneScriptingFloatChannel}
          Translation.Y {type=MovieSceneScriptingFloatChannel}
          Angle {type=MovieSceneScriptingFloatChannel}
          Scale.X {type=MovieSceneScriptingFloatChannel, const [2,1]}
          Scale.Y {type=MovieSceneScriptingFloatChannel, const [2,1]}
          Shear.X {type=MovieSceneScriptingFloatChannel}
          Shear.Y {type=MovieSceneScriptingFloatChannel}
         ]
        }]
      }
      {
       type=MovieSceneFloatTrack
       display_name=Render Opacity
       sections [{type=MovieSceneFloatSection, end_frame=105, start_frame=0, channels [None {type=MovieSceneScriptingFloatChannel, const [0,1]}]}]
      }
     ]
    }
    Crew_Subheading {
     id=#9
     possessed_class=CommonTextBlock
     tracks [
      {
       type=MovieScene2DTransformTrack
       sections [{
         type=MovieScene2DTransformSection
         end_frame=105
         start_frame=0
         channels [
          Translation.X {type=MovieSceneScriptingFloatChannel}
          Translation.Y {type=MovieSceneScriptingFloatChannel, keys [[47,-40],[54,0]]}
          Angle {type=MovieSceneScriptingFloatChannel}
          Scale.X {type=MovieSceneScriptingFloatChannel}
          Scale.Y {type=MovieSceneScriptingFloatChannel}
          Shear.X {type=MovieSceneScriptingFloatChannel}
          Shear.Y {type=MovieSceneScriptingFloatChannel}
         ]
        }]
      }
      {
       type=MovieSceneColorTrack
       display_name=Color and Opacity
       sections [{
         type=MovieSceneColorSection
         end_frame=105
         start_frame=0
         channels [
          Color.R {type=MovieSceneScriptingFloatChannel, const [0.5457,2]}
          Color.G {type=MovieSceneScriptingFloatChannel, const [0.6105,2]}
          Color.B {type=MovieSceneScriptingFloatChannel, const [1,2]}
          Color.A {type=MovieSceneScriptingFloatChannel, keys [[47,0],[54,1],[92,1],[95,0]]}
         ]
        }]
      }
     ]
    }
    ContentBox {
     id=#10
     possessed_class=Overlay
     tracks [{
       type=MovieScene2DTransformTrack
       sections [{
         type=MovieScene2DTransformSection
         end_frame=105
         start_frame=0
         channels [
          Translation.X {type=MovieSceneScriptingFloatChannel, const [0,1]}
          Translation.Y {type=MovieSceneScriptingFloatChannel, const [0,1]}
          Angle {type=MovieSceneScriptingFloatChannel, const [0,1]}
          Scale.X {type=MovieSceneScriptingFloatChannel, keys [[0,1],[86,1],[105.019,0.5]]}
          Scale.Y {type=MovieSceneScriptingFloatChannel, keys [[0,1],[86,1],[105.019,0.5]]}
          Shear.X {type=MovieSceneScriptingFloatChannel, const [0,1]}
          Shear.Y {type=MovieSceneScriptingFloatChannel, const [0,1]}
         ]
        }]
      }]
    }
    LogoOverlay {
     id=#11
     possessed_class=Overlay
     tracks [{
       type=MovieScene2DTransformTrack
       sections [{
         type=MovieScene2DTransformSection
         end_frame=105
         start_frame=0
         channels [
          Translation.X {type=MovieSceneScriptingFloatChannel}
          Translation.Y {type=MovieSceneScriptingFloatChannel, const [0,1]}
          Angle {type=MovieSceneScriptingFloatChannel}
          Scale.X {type=MovieSceneScriptingFloatChannel, keys [[0,1.8148],[25,1.2],[84,1.2],[100,0.55]]}
          Scale.Y {type=MovieSceneScriptingFloatChannel, keys [[0,1.8148],[25,1.2],[84,1.2],[100,0.55]]}
          Shear.X {type=MovieSceneScriptingFloatChannel}
          Shear.Y {type=MovieSceneScriptingFloatChannel}
         ]
        }]
      }]
    }
    PassBadges {
     id=#12
     possessed_class=HorizontalBox
     tracks [
      {
       type=MovieSceneFloatTrack
       display_name=Render Opacity
       sections [{type=MovieSceneFloatSection, end_frame=105, start_frame=0, channels [None {type=MovieSceneScriptingFloatChannel, const [0,1]}]}]
      }
      {type=MovieScene2DTransformTrack, sections []}
     ]
    }
   ]
   display_rate {numerator=30}
  }
 }
]
graphs [EventGraph {nodes [
   PlayIntro() {is=title K2Node_CustomEvent}
   > PlayAnimation(InAnimation=Get UnlockAnimation_Short() -> (UnlockAnimation_Short:Widget Animation Object Reference))
   > PlaySound2D(Sound=Get SoundOnPlayIntro() -> (SoundOnPlayIntro:Sound Base Object Reference))
   > n1 = Sequence()
   > Activate Control Bus Mix!(execute, Mix:Sound Control Bus Mix Object Reference=Get ControlBusMix() -> (ControlBusMix:Sound Control Bus Mix Object Reference)) -> (then)
   > BindToAnimationFinished(Animation=Get UnlockAnimation_Short(), Delegate=Create Event())
   > Do Once()
   > Delay(Duration=2)
   > SpawnSound2D!(
    execute
    Sound:Sound Base Object Reference=/Game/Sounds/Fort_UI_Sounds/FrontEnd/RewardsAndGifts/UI_FrontEnd_Gift_Screen_Dissapear_Cue
    VolumeMultiplier:float=Get Volume Multiplier() -> (Volume Multiplier:float)
    PitchMultiplier:float=1
    StartTime:float=0
    ConcurrencySettings:Sound Concurrency Object Reference
    bPersistAcrossLevelTransition:bool=false
    bAutoDestroy:bool=true
   ) -> (then, ReturnValue:Audio Component Object Reference)
   Delay(execute=@n1.then_1, Duration=3)
   > Deactivate Control Bus Mix!(execute, Mix:Sound Control Bus Mix Object Reference=Get ControlBusMix()) -> (then)
   OnIntroAnimationFinished() {is=title K2Node_CustomEvent}
   > UnbindFromAnimationFinished!(
    execute
    self:User Widget Object Reference
    Animation:Widget Animation Object Reference=Get UnlockAnimation_Short()
    Delegate:Delegate=Create Event()
   ) -> (then)
   > OnPresentationFinished(self:Athena Upgraded Pre Message Object Reference)
   Event On Activated()
   > PlayIntro() {is=title K2Node_CallFunction}
  ]}]
variables [
 ControlBusMix {type=SoundControlBusMix, default=/Game/Sounds/Fort_UI_Sounds/FrontEnd/RewardsAndGifts/CMB_GiftingDucksMusic}
 SoundOnPlayIntro {type=SoundBase, default=/Game/Sounds/Fort_UI_Sounds/FrontEnd/RewardsAndGifts/UI_FrontEnd_Gift_Unlock_Anim_Cue}
 Volume Multiplier {type=number, default=0.8}
]
widgets {
 info {inheritedWidgetCount=0, parentClass=/Script/FortniteUI.AthenaUpgradedPreMessage, rootWidgetClass=&Overlay, widgetCount=14}
 widgets [
  Overlay_1 {widgetClassPath=&Overlay, widget=Overlay_1, properties {clipping=ClipToBoundsAlways}}
  Crew_Background {
   widgetClassPath=&Image
   bIsVariable=true
   parent=Overlay_1
   widget=Crew_Background
   properties {
    visibility=SelfHitTestInvisible
    brush {
     bIsDynamicallyLoaded=false
     drawAs=Image
     imageType=FullColor
     mirroring=NoMirror
     resourceName=None
     resourceObject=/Crew/UI/Materials/CrewUnlocking/M_UI_Crew_Unlocking_BG
     tiling=NoTile
     imageSize=<32 32>
     margin=<ltrb 0 0 0 0>
     outlineSettings {
      bUseBrushTransparency=false
      roundingType=HalfHeightRadius
      width=0
      color {colorUseRule=UseColor_Specified, specifiedColor=<rgba 0 0 0 0>}
      cornerRadii=<0 0 0 0>
     }
     tintColor {colorUseRule=UseColor_Specified, specifiedColor=<rgba 1 1 1 1>}
     uVRegion {bIsValid=false, max=<0 0>, min=<0 0>}
    }
   }
   slot {class=OverlaySlot, properties {horizontalAlignment=HAlign_Fill, verticalAlignment=VAlign_Fill}}
  }
  ContentBox {
   widgetClassPath=&Overlay
   parent=Overlay_1
   widget=ContentBox
   slot {class=OverlaySlot, properties {horizontalAlignment=HAlign_Center, verticalAlignment=VAlign_Center}}
  }
  LogoOverlay {
   widgetClassPath=&Overlay
   parent=ContentBox
   widget=LogoOverlay
   slot {class=OverlaySlot, properties {horizontalAlignment=HAlign_Center, verticalAlignment=VAlign_Center}}
  }
  Crew_Logo {
   widgetClassPath=&Image
   bIsVariable=true
   parent=LogoOverlay
   widget=Crew_Logo
   properties {brush {
     bIsDynamicallyLoaded=false
     drawAs=Image
     imageType=FullColor
     mirroring=NoMirror
     resourceName=None
     resourceObject=/Crew/UI/Materials/CrewUnlocking/M_UI_Crew_Unlocking_Logo
     tiling=NoTile
     imageSize=<1000 565>
     margin=<ltrb 0 0 0 0>
     outlineSettings {
      bUseBrushTransparency=false
      roundingType=HalfHeightRadius
      width=0
      color {colorUseRule=UseColor_Specified, specifiedColor=<rgba 0 0 0 0>}
      cornerRadii=<0 0 0 0>
     }
     tintColor {colorUseRule=UseColor_Specified, specifiedColor=<rgba 1 1 1 1>}
     uVRegion {bIsValid=false, max=<0 0>, min=<0 0>}
    }}
   slot {class=OverlaySlot, properties {padding=<ltrb 0 -40 0 0>}}
  }
  Crew_Subheading {
   widgetClassPath=&CommonTextBlock
   bIsVariable=true
   parent=LogoOverlay
   widget=Crew_Subheading
   properties {
    bIsScrollingEnabled=false
    text=Member benefits
    textTransformPolicy=ToUpper
    colorAndOpacity {colorUseRule=UseColor_Specified, specifiedColor=<rgba 0.5457 0.6105 1 1>}
    font {
     bForceMonospaced=false
     bMaterialIsStencil=false
     fontMaterial=null
     fontObject=&HeadingNow
     letterSpacing=0
     monospacedWidth=1
     size=25.5
     skewAmount=0
     typefaceFontName=86 Bold Italic
     outlineSettings {
      bApplyOutlineToDropShadows=false
      bMiteredCorners=false
      bSeparateFillAlpha=false
      outlineMaterial=null
      outlineSize=0
      outlineColor=<rgba 0 0 0 1>
     }
    }
    shadowOffset=<0 0>
   }
   slot {class=OverlaySlot, properties {horizontalAlignment=HAlign_Center, verticalAlignment=VAlign_Center, padding=<ltrb 0 160 0 0>}}
  }
  PassBadges {
   widgetClassPath=&HorizontalBox
   parent=ContentBox
   widget=PassBadges
   properties {renderOpacity=0}
   slot {class=OverlaySlot, properties {horizontalAlignment=HAlign_Center, verticalAlignment=VAlign_Center, padding=<ltrb 0 10 0 0>}}
  }
  BP_Badge {
   widgetClassPath=&Image
   bIsVariable=true
   parent=PassBadges
   widget=BP_Badge
   properties {brush {
     bIsDynamicallyLoaded=false
     drawAs=Image
     imageType=FullColor
     mirroring=NoMirror
     resourceName=None
     resourceObject=/Crew/UI/Materials/CrewUnlocking/MI_UI_Crew_Unlocking_Badge_BattlePass
     tiling=NoTile
     imageSize=<256 256>
     margin=<ltrb 0 0 0 0>
     outlineSettings {
      bUseBrushTransparency=false
      roundingType=HalfHeightRadius
      width=0
      color {colorUseRule=UseColor_Specified, specifiedColor=<rgba 0 0 0 0>}
      cornerRadii=<0 0 0 0>
     }
     tintColor {colorUseRule=UseColor_Specified, specifiedColor=<rgba 1 1 1 1>}
     uVRegion {bIsValid=false, max=<0 0>, min=<0 0>}
    }}
   slot {class=HorizontalBoxSlot, properties {horizontalAlignment=HAlign_Center, verticalAlignment=VAlign_Center, padding=<ltrb 40 40 40 40>}}
  }
  Figment_Badge {
   widgetClassPath=&Image
   bIsVariable=true
   parent=PassBadges
   widget=Figment_Badge
   properties {brush {
     bIsDynamicallyLoaded=false
     drawAs=Image
     imageType=FullColor
     mirroring=NoMirror
     resourceName=None
     resourceObject=/Crew/UI/Materials/CrewUnlocking/MI_UI_Crew_Unlocking_Badge_FigmentPass
     tiling=NoTile
     imageSize=<256 256>
     margin=<ltrb 0 0 0 0>
     outlineSettings {
      bUseBrushTransparency=false
      roundingType=HalfHeightRadius
      width=0
      color {colorUseRule=UseColor_Specified, specifiedColor=<rgba 0 0 0 0>}
      cornerRadii=<0 0 0 0>
     }
     tintColor {colorUseRule=UseColor_Specified, specifiedColor=<rgba 1 1 1 1>}
     uVRegion {bIsValid=false, max=<0 0>, min=<0 0>}
    }}
   slot {class=HorizontalBoxSlot, properties {horizontalAlignment=HAlign_Center, verticalAlignment=VAlign_Center, padding=<ltrb 40 40 40 40>}}
  }
  Music_Badge {
   widgetClassPath=&Image
   bIsVariable=true
   parent=PassBadges
   widget=Music_Badge
   properties {brush {
     bIsDynamicallyLoaded=false
     drawAs=Image
     imageType=FullColor
     mirroring=NoMirror
     resourceName=None
     resourceObject=/Crew/UI/Materials/CrewUnlocking/MI_UI_Crew_Unlocking_Badge_MusicPass
     tiling=NoTile
     imageSize=<256 256>
     margin=<ltrb 0 0 0 0>
     outlineSettings {
      bUseBrushTransparency=false
      roundingType=HalfHeightRadius
      width=0
      color {colorUseRule=UseColor_Specified, specifiedColor=<rgba 0 0 0 0>}
      cornerRadii=<0 0 0 0>
     }
     tintColor {colorUseRule=UseColor_Specified, specifiedColor=<rgba 1 1 1 1>}
     uVRegion {bIsValid=false, max=<0 0>, min=<0 0>}
    }}
   slot {class=HorizontalBoxSlot, properties {horizontalAlignment=HAlign_Center, verticalAlignment=VAlign_Center, padding=<ltrb 40 40 40 40>}}
  }
  Lego_Badge {
   widgetClassPath=&Image
   bIsVariable=true
   parent=PassBadges
   widget=Lego_Badge
   properties {brush {
     bIsDynamicallyLoaded=false
     drawAs=Image
     imageType=FullColor
     mirroring=NoMirror
     resourceName=None
     resourceObject=/Crew/UI/Materials/CrewUnlocking/MI_UI_Crew_Unlocking_Badge_LegoPass
     tiling=NoTile
     imageSize=<256 256>
     margin=<ltrb 0 0 0 0>
     outlineSettings {
      bUseBrushTransparency=false
      roundingType=HalfHeightRadius
      width=0
      color {colorUseRule=UseColor_Specified, specifiedColor=<rgba 0 0 0 0>}
      cornerRadii=<0 0 0 0>
     }
     tintColor {colorUseRule=UseColor_Specified, specifiedColor=<rgba 1 1 1 1>}
     uVRegion {bIsValid=false, max=<0 0>, min=<0 0>}
    }}
   slot {class=HorizontalBoxSlot, properties {horizontalAlignment=HAlign_Center, verticalAlignment=VAlign_Center, padding=<ltrb 40 40 40 40>}}
  }
  Slogan_VerticalBox {
   widgetClassPath=&VerticalBox
   parent=ContentBox
   widget=Slogan_VerticalBox
   slot {class=OverlaySlot, properties {horizontalAlignment=HAlign_Center, verticalAlignment=VAlign_Center, padding=<ltrb 0 300 0 0>}}
  }
  Slogantext_1 {
   widgetClassPath=&CommonTextBlock
   bIsVariable=true
   parent=Slogan_VerticalBox
   widget=Slogantext_1
   properties {
    bIsScrollingEnabled=false
    justification=Center
    renderOpacity=0
    text=Every single pass
    textTransformPolicy=ToUpper
    colorAndOpacity {colorUseRule=UseColor_Specified, specifiedColor=<rgba 0.5457 0.6105 1 1>}
    font {
     bForceMonospaced=false
     bMaterialIsStencil=false
     fontMaterial=null
     fontObject=&HeadingNow
     letterSpacing=0
     monospacedWidth=1
     size=25.5
     skewAmount=0
     typefaceFontName=86 Bold Italic
     outlineSettings {
      bApplyOutlineToDropShadows=false
      bMiteredCorners=false
      bSeparateFillAlpha=false
      outlineMaterial=null
      outlineSize=0
      outlineColor=<rgba 0 0 0 1>
     }
    }
    shadowOffset=<0 0>
   }
   slot {class=VerticalBoxSlot}
  }
  SloganText_2 {
   widgetClassPath=&CommonTextBlock
   bIsVariable=true
   parent=Slogan_VerticalBox
   widget=SloganText_2
   properties {
    bIsScrollingEnabled=false
    justification=Center
    renderOpacity=0
    text="Every single time!"
    textTransformPolicy=ToUpper
    colorAndOpacity {colorUseRule=UseColor_Specified, specifiedColor=<rgba 0.7917 0.8214 1 1>}
    font {
     bForceMonospaced=false
     bMaterialIsStencil=false
     fontMaterial=null
     fontObject=&HeadingNow
     letterSpacing=0
     monospacedWidth=1
     size=25.5
     skewAmount=0
     typefaceFontName=86 Bold Italic
     outlineSettings {
      bApplyOutlineToDropShadows=false
      bMiteredCorners=false
      bSeparateFillAlpha=false
      outlineMaterial=null
      outlineSize=0
      outlineColor=<rgba 0 0 0 1>
     }
    }
    shadowOffset=<0 0>
   }
   slot {class=VerticalBoxSlot, properties {padding=<ltrb 0 10 0 0>}}
  }
 ]
}
