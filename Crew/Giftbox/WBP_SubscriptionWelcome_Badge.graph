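# Badge widget for the Crew giftbox subscription welcome, showing the large Battle Pass icon in a single root FortLazyImage.
# Its Badge_Intro animation shrinks the icon from an oversized scale to normal while fading it in over a short beat.
# PreConstruct plays the intro only outside design time, so it runs automatically in game and is skipped in the designer preview.
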WidgetBlueprint /Crew/Giftbox/WBP_SubscriptionWelcome_Badge
animations [Badge_Intro {
  binding_count=1
  length_seconds=0.45
  object=&self:Badge_Intro
  sequence {bindings [PassBadge_Icon {
     id=#1
     possessed_class=FortLazyImage
     tracks [
      {
       type=MovieScene2DTransformTrack
       sections [{
         type=MovieScene2DTransformSection
         end_frame=9
         start_frame=0
         channels [
          Translation.X {type=MovieSceneScriptingFloatChannel, const [0,1]}
          Translation.Y {type=MovieSceneScriptingFloatChannel, const [0,1]}
          Angle {type=MovieSceneScriptingFloatChannel, const [0,1]}
          Scale.X {type=MovieSceneScriptingFloatChannel, keys [[0,2.2],[9,1]]}
          Scale.Y {type=MovieSceneScriptingFloatChannel, keys [[0,2.2],[9,1]]}
          Shear.X {type=MovieSceneScriptingFloatChannel, const [0,1]}
          Shear.Y {type=MovieSceneScriptingFloatChannel, const [0,1]}
         ]
        }]
      }
      {
       type=MovieSceneFloatTrack
       display_name=Render Opacity
       sections [{type=MovieSceneFloatSection, end_frame=9, start_frame=0, channels [None {type=MovieSceneScriptingFloatChannel, keys [[0,0],[9,1]]}]}]
      }
     ]
    }]}
 }]
graphs [EventGraph {nodes [
   n1 = Event PreConstruct()
   > n2 = Branch(Condition=@n1.IsDesignTime)
   PlayAnimation(execute=@n2.else, InAnimation=Get Badge_Intro() -> (Badge_Intro:Widget Animation Object Reference))
  ]}]
variables []
widgets {
 info {inheritedWidgetCount=0, parentClass=&UserWidget, rootWidgetClass=&FortLazyImage, widgetCount=1}
 widgets [PassBadge_Icon {
   widgetClassPath=&FortLazyImage
   bIsVariable=true
   widget=PassBadge_Icon
   properties {brush {
     bIsDynamicallyLoaded=false
     drawAs=Image
     imageType=FullColor
     mirroring=NoMirror
     resourceName=None
     resourceObject=/Game/UI/Foundation/Textures/Icons/Items/T-FNBR-BattlePass-L
     tiling=NoTile
     imageSize=<256 256>
     margin=<ltrb 0 0 0 0>
     outlineSettings {
      bUseBrushTransparency=false
      roundingType=HalfHeightRadius
      width=0
      color {colorUseRule=UseColor_Specified, specifiedColor=<rgba 0 0 0 0>}
      cornerRadii=<0 0 0 0>
     }
     tintColor {colorUseRule=UseColor_Specified, specifiedColor=<rgba 1 1 1 1>}
     uVRegion {bIsValid=false, max=<0 0>, min=<0 0>}
    }}
  }]
}
